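# CustomizableHUD is the widget Blueprint for Fortnite's mobile customizable HUD, built on the native FortMobileCustomizableHUD class.
# A root overlay stretches the inherited HUD widget layer to fill it, letting input pass through the layer itself to its children.
# A vertical playlist UI extension slot, anchored bottom-right, provides the MobileHUDBottomRight extension point.
# The event graph holds only unconnected PreConstruct, Construct and Tick events, so the Blueprint adds layout without logic.
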WidgetBlueprint /FortniteCoreUI/HUD/Bacchus/CustomizableHUD/CustomizableHUD
graphs [EventGraph {nodes [
   Event PreConstruct()
   Event Construct()
   Event Tick()
  ]}]
variables []
widgets {
 info {inheritedWidgetCount=1, parentClass=/Script/FortniteUI.FortMobileCustomizableHUD, rootWidgetClass=&Overlay, widgetCount=3}
 widgets [
  Overlay_36 {widgetClassPath=&Overlay, widget=Overlay_36}
  PlaylistUIExtensionSlot_BottomRight {
   widgetClassPath=/Script/FortniteUI.FortPlaylistUIExtensionSlot
   bIsVariable=true
   parent=Overlay_36
   widget=PlaylistUIExtensionSlot_BottomRight
   properties {entryBoxType=Vertical, entryWidgetClass=&CommonUserWidget, extensionSlots [MobileHUDBottomRight]}
   slot {class=OverlaySlot, properties {horizontalAlignment=HAlign_Right, verticalAlignment=VAlign_Bottom}}
  }
  HUDLayer_HUDWidgets {
   widgetClassPath=/Script/FortniteUI.FortMobileHUDLayer
   bInherited=true
   bIsVariable=true
   parent=Overlay_36
   widget=HUDLayer_HUDWidgets
   properties {visibility=SelfHitTestInvisible}
   slot {class=OverlaySlot, properties {horizontalAlignment=HAlign_Fill, verticalAlignment=VAlign_Fill}}
  }
 ]
}
