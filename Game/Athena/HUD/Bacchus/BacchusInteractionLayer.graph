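# BacchusInteractionLayer is a touch interaction layer widget for the Athena Bacchus HUD, derived from FortTouchInteractionLayer.
# An Overlay root stacks a visible FortInWorldInteractionLayer, set to fill, beneath the inherited ContentPanel canvas.
# The event graph holds only unconnected PreConstruct, Construct and Tick events, so the blueprint supplies layout and adds no logic of its own.
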WidgetBlueprint /Game/Athena/HUD/Bacchus/BacchusInteractionLayer
graphs [EventGraph {nodes [
   Event PreConstruct()
   Event Construct()
   Event Tick()
  ]}]
variables []
widgets {
 info {inheritedWidgetCount=1, parentClass=/Script/FortniteUI.FortTouchInteractionLayer, rootWidgetClass=&Overlay, widgetCount=3}
 widgets [
  Overlay_1 {widgetClassPath=&Overlay, widget=Overlay_1}
  FortInWorldInteractionLayer {
   widgetClassPath=&/FortInWorldInteractionLayer
   bIsVariable=true
   parent=Overlay_1
   widget=FortInWorldInteractionLayer
   properties {visibility=Visible}
   slot {class=OverlaySlot, properties {horizontalAlignment=HAlign_Fill, verticalAlignment=VAlign_Fill}}
  }
  ContentPanel {widgetClassPath=&CanvasPanel, bInherited=true, parent=Overlay_1, widget=ContentPanel, slot {class=OverlaySlot}}
 ]
}
