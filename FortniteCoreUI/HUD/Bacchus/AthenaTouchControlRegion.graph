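# Touch control region widget for the Athena mobile (Bacchus) HUD, hosting the interaction layer, tutorial container and RC pawn expiration bar.
# On construct it asynchronously creates the CustomizableHUD widget and fills the HUD content overlay with it.
# It also acts as an inventory drop target: drag-over shows a drop icon based on droppability.
# Dropping discards the full stack, or half (minimum one) when splitting, through the inventory context.
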WidgetBlueprint /FortniteCoreUI/HUD/Bacchus/AthenaTouchControlRegion
graphs [
 EventGraph {nodes [
   Event Construct()
   > n1 = CreateWidgetAsync!(
    execute:Exec
    UserWidgetSoftClass:User Widget Soft Class Reference=Get CustomizableHUDClass() -> (CustomizableHUDClass:User Widget Soft Class Reference)
    OwningPlayer:Player Controller Object Reference=GetOwningPlayer()
    bSuspendInputUntilComplete:bool=true
   ) -> (then:Exec, OnComplete:Exec, OnFailure:Exec, UserWidget:User Widget Object Reference) {is=title K2Node_AsyncAction}
   n2 = AddChild(execute=@n1.OnComplete, self=Get HUD_Content() -> (HUD_Content:Overlay Object Reference), Content=@n1)
   > n3 = Cast To OverlaySlot!(execute, Object=@n2) -> (then, CastFailed, AsOverlay Slot:Overlay Slot Object Reference) {is=title K2Node_DynamicCast}
   > SetHorizontalAlignment(self=@n3, InHorizontalAlignment=HAlign_Fill)
   > SetVerticalAlignment(self=@n3, InVerticalAlignment=VAlign_Fill)
  ]}
 IsDragItemDroppable {nodes [
   n1 = IsDragItemDroppable!() -> (then, DragDropOperation:Drag Drop Operation Object Reference) {is=title K2Node_FunctionEntry}
   > n2 = Cast To AthenaInventoryDragDrop!(execute, Object=@n1) -> (then, CastFailed, AsAthena Inventory Drag Drop:Athena Inventory Drag Drop Object Reference) {is=title K2Node_DynamicCast}
   > n3 = Cast To FortWorldItem!(execute, Object=Get Payload!(self:Drag Drop Operation Object Reference=@n2) -> (Payload:Object Reference)) -> (then, CastFailed, AsFort World Item:Fort World Item Object Reference) {is=title K2Node_DynamicCast}
   > Return Node!(execute, bIsDroppable:bool=CanBeDropped!(self:Fort Item Object Reference=@n3) -> (ReturnValue:bool))
   Return Node(execute=@n3.CastFailed, bIsDroppable=false)
   Return Node(execute=@n2.CastFailed, bIsDroppable=false)
  ]}
 OnDragOver {nodes [
   n1 = OnDragOver!() -> (then, MyGeometry:Geometry Structure, PointerEvent:Pointer Event Structure, Operation:Drag Drop Operation Object Reference) {is=title K2Node_FunctionEntry}
   > n2 = Cast To AthenaInventoryDragDrop(Object=@n1.Operation)
   > ShowDropIcon!(
    execute
    self:Athena Inventory Drag Drop Object Reference=@n2
    Drop Action:AthenaDragDropAction Enum=Select(
     Option 0:AthenaDragDropAction Enum=NewEnumerator3
     Option 1:AthenaDragDropAction Enum=NewEnumerator1
     Index=IsDragItemDroppable!(DragDropOperation:Drag Drop Operation Object Reference=@n2) -> (bIsDroppable:bool) {is=title K2Node_CallFunction}
    ) -> (ReturnValue:AthenaDragDropAction Enum)
   ) -> (then)
   > Return Node!(execute, ReturnValue:bool=false)
   Return Node(execute=@n2.CastFailed, ReturnValue=false)
  ]}
 OnDrop {nodes [
   n1 = OnDrop!() -> (then, MyGeometry:Geometry Structure, PointerEvent:Pointer Event Structure, Operation:Drag Drop Operation Object Reference) {is=title K2Node_FunctionEntry}
   > n2 = Branch(Condition=IsDragItemDroppable(DragDropOperation=@n1.Operation))
   > n3 = Cast To AthenaInventoryDragDrop(Object=@n1.Operation)
   > n4 = Cast To FortWorldItem(Object=Get Payload(self=@n3))
   n5 = GetNumInStack!(self:Fort Item Object Reference=@n4) -> (ReturnValue:int=0)
   n6 = int / int!(A:int=@n5, B:int=2) -> (ReturnValue:int=0) {is=title K2Node_CallFunction}
   > DropItem!(
    execute
    self:Fort Inventory Context Object Reference=Get FortInventoryContext!() -> (ReturnValue:Fort Inventory Context Object Reference) {is=title K2Node_GetBlueprintContextEx}
    ItemBeingDropped:Fort World Item Object Reference=@n4
    Quantity:int=Select(
     Option 0:int=@n5
     Option 1:int="Max (Integer)"(A=@n6, B=1)
     Index=Get SplitTheStack!(self:Athena Inventory Drag Drop Object Reference=@n3) -> (SplitTheStack:bool)
    ) -> (ReturnValue:int)
   ) -> (then, ReturnValue:bool)
   Return Node(execute=[@n4.CastFailed, @n3.CastFailed, @n2.else], ReturnValue=false)
  ]}
]
variables [CustomizableHUDClass {type=Class, default=/FortniteCoreUI/HUD/Bacchus/CustomizableHUD/CustomizableHUD.CustomizableHUD_C}]
widgets {
 info {inheritedWidgetCount=0, parentClass=/Script/FortniteUI.FortTouchControlRegion, rootWidgetClass=&Overlay, widgetCount=6}
 widgets [
  Overlay_0 {widgetClassPath=&Overlay, widget=Overlay_0}
  BacchusInteractionLayer {
   widgetClassPath=/Game/Athena/HUD/Bacchus/BacchusInteractionLayer
   bIsVariable=true
   parent=Overlay_0
   widget=BacchusInteractionLayer
   slot {class=OverlaySlot, properties {horizontalAlignment=HAlign_Fill, verticalAlignment=VAlign_Fill}}
  }
  CanvasPanel_0 {
   widgetClassPath=&CanvasPanel
   parent=Overlay_0
   widget=CanvasPanel_0
   slot {class=OverlaySlot, properties {horizontalAlignment=HAlign_Fill, verticalAlignment=VAlign_Fill}}
  }
  HUD_Content {
   widgetClassPath=&Overlay
   bIsVariable=true
   parent=CanvasPanel_0
   widget=HUD_Content
   slot {class=CanvasPanelSlot, properties {layoutData {alignment=<0 0>, anchors {maximum=<1 1>, minimum=<0 0>}, offsets=<ltrb 0 0 0 0>}}}
  }
  RCPawn_ExpirationProgressBar {
   widgetClassPath=/Game/Athena/UI/InGame/RCPawn/RCPawn_ExpirationProgressBar
   bIsVariable=true
   parent=CanvasPanel_0
   widget=RCPawn_ExpirationProgressBar
   properties {visibility=Collapsed, elementTag {tagName=Hud.Athena.HideOnCinematic}}
   slot {class=CanvasPanelSlot, properties {layoutData {alignment=<0.5 1>, anchors {maximum=<0.5 1>, minimum=<0.5 1>}, offsets=<ltrb 0 -200 300 12>}}}
  }
  AthenaTutorialHUDContainer {
   widgetClassPath=/Game/Athena/HUD/Tutorial/AthenaTutorialHUDContainer
   bIsVariable=true
   parent=CanvasPanel_0
   widget=AthenaTutorialHUDContainer
   properties {hUDElementTag {gameplayTags [
      {tagName=HUD.Athena.HideOnSpectate}
      {tagName=Athena.Creative.HideForLoading}
      {tagName=HUD.Hide.WhileLoading}
      {tagName=Hud.Athena.HideOnCinematic}
      {tagName=HUD.Athena.HideForPostGameSpectate}
      {tagName=HUD.Athena.HideForPostGame}
      {tagName=HUD.Bacchus.TutorialHUDContainer}
     ]}}
   slot {class=CanvasPanelSlot, properties {zOrder=10, layoutData {alignment=<0 0>, anchors {maximum=<1 1>, minimum=<0 0>}, offsets=<ltrb 0 0 0 0>}}}
  }
 ]
}
