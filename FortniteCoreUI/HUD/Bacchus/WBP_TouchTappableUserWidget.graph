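# WBP_TouchTappableUserWidget is a Widget Blueprint subclass of the native TouchTappableUserWidget for the HUD that shows only on touch input.
# Its root CommonHardwareVisibilityBorder uses a visibility query requiring the Input.Touch tag and wraps an empty, visible Overlay with zero padding.
# The event graph holds only unconnected PreConstruct, Construct and Tick events, so any tap behavior is left to the parent class.
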WidgetBlueprint /FortniteCoreUI/HUD/Bacchus/WBP_TouchTappableUserWidget
graphs [EventGraph {nodes [
   Event PreConstruct()
   Event Construct()
   Event Tick()
  ]}]
variables []
widgets {
 info {
  inheritedWidgetCount=0
  parentClass=/Script/FortniteUI.TouchTappableUserWidget
  rootWidgetClass=&CommonHardwareVisibilityBorder
  widgetCount=2
 }
 widgets [
  show-on-touch-only {
   widgetClassPath=&CommonHardwareVisibilityBorder
   widget=show-on-touch-only
   properties {visibilityQuery {
     autoDescription=" ALL( Input.Touch )"
     tokenStreamVersion=0
     userDescription=show-on-touch-only
     queryTokenStream [0,1,2,1,0]
     tagDictionary [{tagName=Input.Touch}]
    }}
  }
  Overlay_30 {
   widgetClassPath=&Overlay
   parent=show-on-touch-only
   widget=Overlay_30
   properties {visibility=Visible}
   slot {class=BorderSlot, properties {padding=<ltrb 0 0 0 0>}}
  }
 ]
}
